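# Idle Niagara effect for the BoldDormMum back bling: warm amber glows attached to the backpack's skeletal mesh.
# One burst spawns five persistent sprites, with partition index selecting each one's socket, size and camera offset scale.
# Camera direction against the owner's Y axis drives offset, fade and brightness, so glows show from one side.
# Front-end flags adjust size and opacity; a partition-masked light renderer adds local lighting.
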NiagaraSystem /BRCosmetics/Accessories/FORT_Backpacks/Backpack_M_MED_BoldDormMum/FX/NS_Backpack_BoldDormMum_Idle
dependencies {
 usedDataInterfaces [
  &NiagaraDataInterfaceArrayFloat2
  &NiagaraDataInterfaceArrayFloat
  &NiagaraDataInterfaceCurve
  &NiagaraDataInterfaceCamera
  &NiagaraDataInterfaceSkeletalMesh
 ]
 usedDynamicInputs []
 usedModules [
  &self:SystemSpawnScript.SystemScriptSource.SystemScriptGraph.NiagaraNodeAssignment_0.SetVariables_#1
  &SystemState
 ]
}
emitters {Glows {
  emitterUpdateScript [
   &EmitterState()
   &SpawnBurst_Instantaneous(Spawn Count=5, Use Spawn Probability=false, UseLoopCountLimit=true)
  ]
  particleSpawnScript [
   &PartitionParticles(Number Of Partitions=5)
   &InitializeParticle(
    Color=&MakeCustomLinearColorFromBool(Bool=$User.bIsFrontEndPreview, True Color=<rgba 0.6 0.2824 0.0392 0.7>, False Color=<rgba 0.6 0.2824 0.0392 1>)
    Lifetime=5
    Sprite Size=&FN_FrontEndVector2Modifier(
     Front End:Vector2f=&SelectVector2DFromArray(
      Direct Array Index=$Particles.PartitionParticles.Partition
      Vector 2D Selection Array=DataInterface() {propertyValues {FloatData [<12 5>, <12 5>, <6.5 6.5>, <15 30>, <6.5 6.5>]}}
     )
     Gameplay:Vector2f=&SelectVector2DFromArray(
      Direct Array Index=$Particles.PartitionParticles.Partition
      Vector 2D Selection Array=DataInterface() {propertyValues {FloatData [<12 5>, <12 5>, <6.5 6.5>, <15 25>, <6.5 6.5>]}}
     )
    )
    Sprite Size Mode=::NewEnumerator4
    UsePositionOffset=false
    Write Material Random=true
   )
  ]
  particleUpdateScript [
   &ParticleState(Kill Particles When Lifetime Has Elapsed=false, Let Infinitely Lived Particles Die When Emitter Deactivates=true)
   &DynamicMaterialParameters(
    Index 0 Param 1=8
    Module.Index0Float0Write=true
    Module.Index0Float1Write=true
    Module.Index0Float2Write=true
    Module.Index0Float3Write=true
   )
   &SkeletalMeshLocation(
    Bone / Socket Index=$Particles.PartitionParticles.Partition
    MaskSpawnGroup=false
    Mesh Sampling Type=::NewEnumerator5
    Skeletal Mesh=$System.Skeletal Mesh
    Socket Sampling Mode=::NewEnumerator5
   )
   &CameraOffset(Camera Offset Amount=&FloatFromCurve(
     FloatCurve=DataInterface() {propertyValues {Curve {keys [
         {interpMode=RCIM_Cubic, tangentMode=RCTM_User, time=-1, value=-5}
         {arriveTangent=-0.0131, interpMode=RCIM_Cubic, leaveTangent=-0.0131, tangentMode=RCTM_User, time=-0.1, value=-5}
         {arriveTangent=5.8183, tangentMode=RCTM_User, time=0, value=10}
         <at 1 5>
        ]}}}
     CurveIndex=&DotProduct(
      Vector A=&ReturnCameraVector(Camera Vectors=::NewEnumerator1, Coordinate Space=::Simulation)
      Vector B=$Engine.Owner.SystemYAxis
     )
     Scale Curve=&SelectFloatFromArray(
      Direct Array Index=$Particles.PartitionParticles.Partition
      Float Selection Array=DataInterface() {propertyValues {FloatData [0.1,0.1,0.2,1,0.2]}}
     )
    ))
   &ScaleColor(
    Scale Alpha=&FloatFromCurve(
     FloatCurve=DataInterface() {propertyValues {Curve {keys [
         <at -1 0>
         {interpMode=RCIM_Cubic, tangentMode=RCTM_User, time=-0.1, value=0}
         {interpMode=RCIM_Cubic, tangentMode=RCTM_User, time=0.1, value=1}
         <at 1 1>
        ]}}}
     CurveIndex=&DotProduct(
      Vector A=&ReturnCameraVector(Camera Vectors=::NewEnumerator1, Coordinate Space=::Simulation)
      Vector B=$Engine.Owner.SystemYAxis
     )
    )
    Scale RGB=&Multiply_VectorByFloat(
     Vector=<1 1 1>
     Float=&FloatFromCurve(
      FloatCurve=DataInterface() {propertyValues {Curve {keys [
          <at -1 1>
          {interpMode=RCIM_Cubic, tangentMode=RCTM_User, time=0, value=1}
          {interpMode=RCIM_Cubic, tangentMode=RCTM_User, time=0.25, value=1}
          <at 1 3>
         ]}}}
      CurveIndex=&DotProduct(
       Vector A=&ReturnCameraVector(Camera Vectors=::NewEnumerator1, Coordinate Space=::Simulation)
       Vector B=$Engine.Owner.SystemYAxis
      )
     )
    )
    ScaleA=true
    ScaleRGB=true
   )
   &SpriteFacingAndAlignment(
    Alignment Coordinate Space=::Local
    Sprite Alignment=&Lerp_Vector(
     A=<0 0 1>
     B=<1 0 0>
     Alpha=&Abs_Float(Float:NiagaraFloat=&DotProduct(Vector A=&ReturnCameraVector(Camera Vectors=::NewEnumerator1, Coordinate Space=::Simulation), Vector B=<0 0 1>))
    )
    Sprite Facing=<0 1 0>
    WriteAlignment=true
    WriteFacing=false
   )
   &Light_Attributes(
    Color=1
    Color Mode=0
    Diffuse Scale=1
    Enabled=1
    Falloff Exponent=1
    Light Color:LinearColor=<rgba 1 0.4236 0.0808 1>
    Light Diffuse Scale=3
    Light Enabled:NiagaraBool=/Niagara/DynamicInputs/Helpers/MaskBool(
     Mask Source:NiagaraInt32=1
     Mask Parameter:NiagaraInt32=$Particles.PartitionParticles.Partition
     Mask:NiagaraInt32=4
     Masked Bool:NiagaraBool=true
     Passthrough Bool:NiagaraBool=false
    )
    Light Exponent=4
    Light Radius=30
    Light Specular Scale=0
    Radius=1
    Specular Scale=1
   )
   /SetVariables(Particles.LIghtPos:NiagaraPosition=&AddVectorToPosition(
     Position=$Particles.Position
     Vector=&AddVectorToPosition(Position=&Multiply_Vector(A=<0 -12 0>, B=$Engine.Owner.SystemYAxis), Vector=<0 0 5>)
    )) {isSetParameters=true}
  ]
  rendererClasses [&NiagaraSpriteRendererProperties, &NiagaraLightRendererProperties]
  renderers [
   {rendererClass=&NiagaraSpriteRendererProperties}
   {rendererClass=&NiagaraLightRendererProperties, rendererIndex=1}
  ]
 }}
summary {
 systemName=NS_Backpack_BoldDormMum_Idle
 emitters [{emitterName=Glows, rendererClasses [&NiagaraSpriteRendererProperties, &NiagaraLightRendererProperties]}]
 userVariables [
  User.bIsFrontEnd {type=NiagaraBool}
  User.bIsFrontEndPreview {type=NiagaraBool}
 ]
}
