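# Animation Blueprint for the BoldDormMum backpack that shifts and tilts the pack bone procedurally as the spine bends.
# Each update it reads the spine_05 delta from reference pose relative to spine_01 on the owning skeletal mesh component.
# Clamped range mappings turn that rotation into offsets per bend direction, with some ranges summed for a piecewise response.
# Three chained additive Transform (Modify) Bone nodes apply these offsets over the input pose.
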AnimBlueprint /BRCosmetics/Accessories/FORT_Backpacks/Backpack_M_MED_BoldDormMum/Meshes/M_MED_BoldDormMum_Pack_AnimBP
target_skeleton=&/M_MED_BoldDormMum_Pack_Skeleton
graphs [
 AnimGraph {nodes [
   n1 = "Transform (Modify) Bone - Bone: pack"(ComponentPose=Local To Component(LocalPose=Input Pose()), Translation=Get backpack_translation(), Rotation=Get backpack_rotate()) {rotationMode=BMM_Additive, translationMode=BMM_Additive, alphaBoolBlend {blendOption=Linear}, boneToModify {boneName=pack}}
   n2 = "Transform (Modify) Bone - Bone: pack"(
    ComponentPose=@n1
    Translation=Get Backpack_translation_X() -> (Backpack_translation_X:Vector)
    Rotation=Get Backpack_rotate_X() -> (Backpack_rotate_X:Rotator)
   ) {rotationMode=BMM_Additive, translationMode=BMM_Additive, alphaBoolBlend {blendOption=Linear}, boneToModify {boneName=pack}}
   n3 = "Transform (Modify) Bone - Bone: pack"(ComponentPose=@n2, Translation="0,0,0", Rotation=Get Backpack_rotate_-X() -> (Backpack_rotate_-X:Rotator)) {rotationMode=BMM_Additive, translationMode=BMM_Additive, alphaBoolBlend {blendOption=Linear}, boneToModify {boneName=pack}}
   Output Pose(Result=Component To Local(ComponentPose=@n3)) {layerGroup=None}
  ]}
 EventGraph {nodes [
   n1 = Event BlueprintUpdateAnimation()
   n2 = GetOwnerSkeletalMeshComponent()
   n3 = BreakRotator(InRot=BreakTransform(InTransform=GetDeltaTransformFromRefPose(self=@n2, BoneName=spine_05, BaseName=spine_01)).Rotation)
   n4 = MapRangeClamped(Value=@n3.Yaw, InRangeA=-30, InRangeB=-150, OutRangeB=230)
   n5 = FInterpTo(
    Current=Get!() -> (BackpackRotate_Roll:float, BackpackRotate_Pitch:float, BackpackRotate_Yaw:float)
    Target=MapRangeClamped(Value=@n3.Yaw, InRangeA=-8, InRangeB=-30, OutRangeB=10)
    DeltaTime=@n1.DeltaTimeX
    InterpSpeed=50
   )
   > Set backpack_rotate(backpack_rotate=MakeRotator(Roll=float + float(A=@n5, B=@n4)))
   n6 = MapRangeClamped(Value=@n3.Yaw, InRangeA=0, InRangeB=-20, OutRangeB=-3)
   n7 = MapRangeClamped(Value=@n3.Yaw, InRangeA=-20, InRangeB=-50, OutRangeB=-17)
   > Set backpack_translation(backpack_translation=MakeVector(Y=float + float(A=@n6, B=@n7), Z=MapRangeClamped(Value=@n3.Yaw, InRangeA=-30, InRangeB=-50, OutRangeB=-4)))
   > Set Backpack_rotate_X(Backpack_rotate_X:Rotator=MakeRotator(
     Pitch=MapRangeClamped(Value=@n3, InRangeA=10, InRangeB=45, OutRangeB=7)
     Yaw=MapRangeClamped(Value=@n3, InRangeA=10, InRangeB=45, OutRangeB=-20)
    )) -> (Output_Get:Rotator)
   > Set Backpack_translation_X(Backpack_translation_X:Vector=MakeVector(X=MapRangeClamped(Value=@n3, InRangeA=10, InRangeB=45, OutRangeB=-5), Y=0)) -> (Output_Get:Vector)
   > Set Backpack_rotate_-X(Backpack_rotate_-X:Rotator=MakeRotator(
     Roll=MapRangeClamped(Value=@n3, InRangeA=-10, InRangeB=-40, OutRangeB=18)
     Yaw=MapRangeClamped(Value=@n3, InRangeA=-10, InRangeB=-45, OutRangeB=6)
    )) -> (Output_Get:Rotator)
   TryGetPawnOwner()
   FInterpTo(Current=Get().BackpackRotate_Pitch, Target=0, DeltaTime=0, InterpSpeed=50)
  ]}
]
variables [
 Backpack {type=SkeletalMeshComponent}
 backpack_rotate {type=Rotator}
 Backpack_rotate_-X {type=Rotator}
 Backpack_rotate_X {type=Rotator}
 backpack_translation {type=Vector}
 Backpack_translation_X {type=Vector}
 BackpackRotate {type=Rotator}
 Body {type=SkeletalMeshComponentBudgeted}
 Delta Time X {type=number}
 DivingDown Alpha {type=number}
]
